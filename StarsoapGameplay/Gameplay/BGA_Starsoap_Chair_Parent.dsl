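; Parent gameplay actor for the Starsoap sit-down chair/couch; silently dies when hotfix-disabled.
; Interacting claims a SeatComponent seat and adds a player component that drives sitting; crouching blocks or cancels it.
; Seat events update a nearby linked TV and client camera collision; a full couch grants a bonus effect and music stinger, removed on any exit.
; A server-side sphere applies an effect exempting nearby players from teaming anticheat kicks.
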(event EventEndPlay
  (params (EndPlayReason :type EEndPlayReason%20Enum :default "Destroyed"))
  (bind _appliedteamingeffects (Variables|Dynamic|GetAppliedTeamingEffects))
  ;[0.15 0.01118 0.001047 0.5] [EndPlay] Remove all applied PreventTeamingAnticheat effects
  (Networking|SwitchHasAuthority
    (:Authority
      (bind (_keypin _valuepin) (Utilities|Map|ForEachLoop(Map) _appliedteamingeffects)
        (:then
          (Utilities|IsValid _keypin
            (:Is%20Valid
              (GameplayEffects|RemoveActiveGameplayEffect _keypin _valuepin -1))))
        (:CompletedPin
          (Utilities|Map|Clear _appliedteamingeffects))))
    (:Remote
      (for _array_element (Fortnite|GetLocalFortPlayerControllers true)
        (Utilities|IsValid (Game|Player|GetPlayerPawn _array_element)
          (:Is%20Valid
            (Pawn|UnbindEventfromOnPawnCrouchChanged (Game|Player|GetPlayerPawn _array_element) (EventDispatchers|CreateEvent :!event Custom|OnPawnCrouchChanged))))))))

(event EventBeginPlay
  ;[0.1438 0.0426 0.15 0.5] [BeginPlay]
  (if (ScalableFloat|GetValueasBool (Variables|Config|GetHF_Enabled))
    (CallFunction|UpdateNoBuildZone)
    (Networking|SwitchHasAuthority
      (:Authority
        (Collision|AssignOnComponentBeginOverlap (Variables|Default|GetTeamingSphere) (EventDispatchers|CreateEvent :!event Custom|OnTeamingSphereBeginOverlap))
        (Collision|AssignOnComponentEndOverlap (Variables|Default|GetTeamingSphere) (EventDispatchers|CreateEvent :!event Custom|OnTeamingSphereEndOverlap))
        (Components|Sphere|SetSphereRadius (Variables|Default|GetTeamingSphere) (Variables|Config|GetPreventTeamingAnticheatRadius) true) ; Allow radius to be hotfixed
        (CallFunction|GetConnectedTV))
      (:Remote
        (Components|DestroyComponent (Variables|Default|GetTeamingSphere)) ; Not used on client
      ))
    (else
      (Gameplay|SilentDie false))))

(event Custom|PlayMusicStinger
  (params (Player%20State :type Fort%20Player%20State%20Athena%20Object%20Reference))
  ;[0.1043 1 0.854 0.5] Play Music Stinger
  (Networking|CanExecuteCosmeticEvents
    (:True
      (if (Pawn|IsLocalPlayerController (Fortnite|GetFortPlayerControllerfromActor (Actor|GetOwner Player%20State)))
        (Audio|AddEvent (GameInstanceSubsystems|GetMusicEventSubsystem) "/WildEstate/Audio/Gags/SC_IP_VoidSlide_Music.SC_IP_VoidSlide_Music" "(TagName=\"MusicEventSubsystem.Event.POI\")" "(GameplayTags=)")))))

(event Custom|SetUpSeatEventBindings
  (bind _seatcomponent (Variables|Default|GetSeatComponent))
  ;[0 1 0 0.5 16] [SetUpSeatBindings] Called by SeatComponent's BeginPlay, so that we immediately bind to its delegates.  Doing it in OUR BeginPlay is too late, because the SeatComponent's BeginPlay occurs FIRST, and it immediately broadcasts delegates on CLIENT.
  (Default|BindEventtoOnEnter _seatcomponent (EventDispatchers|CreateEvent :!event Custom|HandleEnter))
  (Default|BindEventtoOnExit _seatcomponent (EventDispatchers|CreateEvent :!event Custom|HandleExit))
  (Default|BindEventtoOnEmpty _seatcomponent (EventDispatchers|CreateEvent :!event Custom|HandleEmpty))
  (Default|BindEventtoOnFull _seatcomponent (EventDispatchers|CreateEvent :!event Custom|HandleFull)))

(event Interaction|EventBlueprintOnInteract
  (params (InteractingPawn :type Fort%20Pawn%20Object%20Reference) (InteractionBeingAttempted :type EInteractionBeingAttempted%20Enum :default "FirstInteraction"))
  ;[0 0 0 0.5] [BlueprintOnInteract]
  (Utilities|IsValid InteractingPawn
    (:Is%20Valid
      (Utilities|IsValid (Actor|GetComponentbyClass InteractingPawn "/StarsoapGameplay/Gameplay/BP_Starsoap_PlayerComponent.BP_Starsoap_PlayerComponent_C")
        (:Is%20Valid
          (Fortnite|FortShippingLog self (Utilities|String|Append "[OnInteract] Player [" (Fortnite|GetPlayerNameSafe 0) "] already has PawnComponent") true))
        (:Is%20Not%20Valid
          (bind _asfort_player_state_athena (Utilities|Casting|CastToFortPlayerStateAthena (Variables|/Script/Engine.Pawn|GetPlayerState InteractingPawn)))
          (bind _assignedseatindex (Default|TryOccupy (Variables|Default|GetSeatComponent) _asfort_player_state_athena)) ; Claim a seat in the SeatComponent
          (if (>= _assignedseatindex 0)
            (bind _returnvalue (AddComponent|NODE_AddBP_VoidSlide_PlayerComponent-0 :self InteractingPawn)) ; Give the pawn a component to control all of the sit/stand behavior
            (Class|BPStarsoapPlayerComponent|Initialize _returnvalue _assignedseatindex self (Variables|Camera|GetSpawnedCamera)) ; Perform the rest of the initialization process after BeginPlay
            (else
              (Fortnite|FortShippingLog self (Utilities|String|Append "[OnInteract] Player [" (Fortnite|GetPlayerNameSafe InteractingPawn) "] failed TryOccupy") true))))) ; Check if the pawn already has the PawnComponent
    )
    (:Is%20Not%20Valid
      (Fortnite|FortShippingLog self "[OnInteract] InteractingPawn is null" true))))

(event Custom|UpdateCouchCameraCollision
  ;[0 0 1 0.5] [UpdateCouchCameraCollision]
  (Networking|SwitchHasAuthority
    (:Remote
      (Collision|SetCollisionResponsetoChannel (Variables|Debug|GetStaticMeshComponent) "ECC_Camera" (select (Default|HasLocallyControlledPlayers (Variables|Default|GetSeatComponent)) "ECR_Ignore" "ECR_Block")))))

(event Custom|HandleEnter
  (params (SeatIndex :type Integer) (PlayerNetId :type Unique%20Net%20Id%20Structure) (PlayerState :type Fort%20Player%20State%20Athena%20Object%20Reference))
  ;[1 1 0 0.5] [HandleEnter]
  (Networking|SwitchHasAuthority
    (:Authority
      (bind _connectedtv (CallFunction|GetConnectedTV))
      (Utilities|IsValid _connectedtv
        (:Is%20Valid
          (Variables|/WildEstate/Environment/Props/GroupHug/GroupHug_WideTv_A/Blueprints/BP_GroupHug_WideTv_A.BP_GroupHug_WideTv_A_C|SetbVoidSlideActivated true _connectedtv))))
    (:Remote
      (CallFunction|UpdateCouchCameraCollision))))

(event Custom|HandleExit
  (params (SeatIndex :type Integer) (PlayerNetId :type Unique%20Net%20Id%20Structure) (PlayerState :type Fort%20Player%20State%20Athena%20Object%20Reference))
  (bind _bonusgeplayerstates (Variables|Dynamic|GetBonusGEPlayerStates))
  ;[0.5 0 0 0.5] [HandleExit]
  (Networking|SwitchHasAuthority
    (:Authority
      (for _array_element _bonusgeplayerstates
        (Utilities|IsValid _array_element
          (:Is%20Valid
            (GameplayEffects|RemoveActiveGameplayEffectbySourceEffect (Variables|/Script/FortniteGame.FortPlayerState|GetAbilitySystemComponent _array_element) (Variables|Assets|GetGE_CouchFullBonus) (Variables|/Script/FortniteGame.FortPlayerState|GetAbilitySystemComponent _array_element) 1))))
      (Utilities|Array|Clear _bonusgeplayerstates) ; Stop tracking these PlayerStates,\nsince we removed the bonus GE
    )
    (:Remote
      (CallFunction|UpdateCouchCameraCollision))))

(event Custom|HandleFull
  ;[0.5 0.5 1 0.5] [HandleFull] Grant a bonus GE to all seated players while couch is full
  (Networking|SwitchHasAuthority
    (:Authority
      (Utilities|FlowControl|Sequence
        (:then_0
          (for _array_element (Default|GetPlayerStates (Variables|Default|GetSeatComponent))
            (if (and (Utilities|IsValid _array_element :!class /Script/Engine.KismetSystemLibrary) (not (Utilities|Array|ContainsItem (Variables|Dynamic|GetBonusGEPlayerStates) _array_element)))
              (Utilities|Array|Add (Variables|Dynamic|GetBonusGEPlayerStates) _array_element) ; Track which players we applied\nthe bonus GE to, for removal\nwhen couch is no longer full
              (GameplayEffects|FortApplyGameplayEffectToTarget :self (Variables|/Script/FortniteGame.FortPlayerState|GetAbilitySystemComponent _array_element) :GameplayEffectClass (Variables|Assets|GetGE_CouchFullBonus) :TargetOfEffect (Variables|/Script/FortniteGame.FortPlayerState|GetAbilitySystemComponent _array_element) :OptionalSourceObject self :OptionalEffectCauser self :Level 0.0) ; Apply bonus GE to ALL players while couch is full
              (bind _asfort_player_state_athena (Utilities|Casting|CastToFortPlayerStateAthena _array_element))
              (CallFunction|PlayMusicStinger _asfort_player_state_athena) ; we only want the musical stinger to play when the bonus event occurs
            )))
        (:then_1
          (bind _connectedtv (CallFunction|GetConnectedTV))
          (Utilities|IsValid _connectedtv
            (:Is%20Valid
              (Variables|/WildEstate/Environment/Props/GroupHug/GroupHug_WideTv_A/Blueprints/BP_GroupHug_WideTv_A.BP_GroupHug_WideTv_A_C|SetbVoidSlideBonusActivated true _connectedtv))))))))

(event Custom|HandleEmpty
  ;[1 0.5 0 0.5] [HandleEmpty]
  (Networking|SwitchHasAuthority
    (:Authority
      (bind _connectedtv (CallFunction|GetConnectedTV))
      (Utilities|IsValid _connectedtv
        (:Is%20Valid
          (Variables|/WildEstate/Environment/Props/GroupHug/GroupHug_WideTv_A/Blueprints/BP_GroupHug_WideTv_A.BP_GroupHug_WideTv_A_C|SetbVoidSlideActivated false _connectedtv))))))

(event Custom|OnTeamingSphereBeginOverlap
  (params (OverlappedComponent :type Primitive%20Component%20Object%20Reference) (OtherActor :type Actor%20Object%20Reference) (OtherComp :type Primitive%20Component%20Object%20Reference) (OtherBodyIndex :type Integer) (bFromSweep :type Boolean) (SweepResult :type Hit%20Result%20Structure%20%28by%20ref%29))
  ;[0.000347 0.15 0.04696 0.5] [OnTeamingSphereBeginOverlap] Apply a GE to the overlapped pawn, which has a tag to prevent the anticheat system from kicking players due to teaming
  (bind _asfort_player_pawn (Utilities|Casting|CastToFortPlayerPawn OtherActor))
  (Utilities|IsValid (Ability|GetFortAbilitySystemComponent _asfort_player_pawn)
    (:Is%20Valid
      (bind _returnvalue (GameplayEffects|FortApplyGameplayEffectToTarget :self (Ability|GetFortAbilitySystemComponent _asfort_player_pawn) :GameplayEffectClass (Variables|Assets|GetGE_PreventTeamingAnticheat) :TargetOfEffect (Ability|GetFortAbilitySystemComponent _asfort_player_pawn) :Level 0.0))
      (Utilities|Map|Add (Variables|Dynamic|GetAppliedTeamingEffects) (Ability|GetFortAbilitySystemComponent _asfort_player_pawn) _returnvalue))))

(event Custom|OnTeamingSphereEndOverlap
  (params (OverlappedComponent :type Primitive%20Component%20Object%20Reference) (OtherActor :type Actor%20Object%20Reference) (OtherComp :type Primitive%20Component%20Object%20Reference) (OtherBodyIndex :type Integer))
  (bind _appliedteamingeffects (Variables|Dynamic|GetAppliedTeamingEffects))
  ;[0.15 0.07142 0.04298 0.5] [OnTeamingSphereBeginOverlap] Remove the teaming GE from the player
  (bind _asfort_player_pawn (Utilities|Casting|CastToFortPlayerPawn OtherActor))
  (Utilities|IsValid (Ability|GetFortAbilitySystemComponent _asfort_player_pawn)
    (:Is%20Valid
      (if (.ReturnValue (Utilities|Map|Find _appliedteamingeffects (Ability|GetFortAbilitySystemComponent _asfort_player_pawn)))
        (GameplayEffects|RemoveActiveGameplayEffect (Ability|GetFortAbilitySystemComponent _asfort_player_pawn) (Utilities|Map|Find _appliedteamingeffects (Ability|GetFortAbilitySystemComponent _asfort_player_pawn)) -1)
        (Utilities|Map|Remove (Variables|Dynamic|GetAppliedTeamingEffects) (Ability|GetFortAbilitySystemComponent _asfort_player_pawn))))))

(event Interaction|EventOnLocalInteractStart
  (params (FortPC :type Fort%20Player%20Controller%20Object%20Reference))
  (bind _returnvalue (Game|Player|GetPlayerPawn FortPC))
  ;[0 0 0 0.5] [OnLocalInteract] Cancel interation ability if we crouch while interacting.  This is a precaution to avoid misalignment when sitting while crouching.
  (Utilities|IsValid _returnvalue
    (:Is%20Valid
      (if (Variables|/Script/Engine.Character|GetbIsCrouched _returnvalue)
        (interact|StopItemInteractionAbility _returnvalue "Interrupted" "(TagName=\"\")")
        (else
          (Pawn|UnbindEventfromOnPawnCrouchChanged _returnvalue (EventDispatchers|CreateEvent :!event Custom|OnPawnCrouchChanged))
          (Pawn|BindEventtoOnPawnCrouchChanged _returnvalue (EventDispatchers|CreateEvent :!event Custom|OnPawnCrouchChanged)))))))

(event Custom|OnPawnCrouchChanged
  (params (Crouched :type Boolean) (Pawn :type Fort%20Pawn%20Object%20Reference))
  ;[0 0 0 0.5] [OnLocalInteract] Cancel interation ability if we crouch while interacting.  This is a precaution to avoid misalignment when sitting while crouching.
  (Utilities|IsValid Pawn
    (:Is%20Valid
      (bind _asfort_player_pawn (Utilities|Casting|CastToFortPlayerPawn Pawn))
      (if Crouched
        (interact|StopItemInteractionAbility _asfort_player_pawn "Interrupted" "(TagName=\"\")")))))

(event Interaction|EventOnLocalInteractStop
  (params (FortPC :type Fort%20Player%20Controller%20Object%20Reference))
  (bind _returnvalue (Game|Player|GetPlayerPawn FortPC))
  ;[0 0 0 0.5] [OnLocalInteract] Cancel interation ability if we crouch while interacting.  This is a precaution to avoid misalignment when sitting while crouching.
  (Utilities|IsValid _returnvalue
    (:Is%20Valid
      (Pawn|UnbindEventfromOnPawnCrouchChanged _returnvalue (EventDispatchers|CreateEvent :!event Custom|OnPawnCrouchChanged)))))

(fn BlueprintCanInteract
  (params (InteractingPawn :type Fort%20Pawn%20Object%20Reference) (InteractionBeingAttempted :type EInteractionBeingAttempted%20Enum :default "FirstInteraction") (InteractionType :type TInteractionType%20Enum :default "IT_NoInteraction"))
  (Utilities|FlowControl|Sequence
    (:then_0
      (if (not (Variables|Config|GetbBlockVoidSlideInteract))
        (Utilities|IsValid InteractingPawn
          (:Is%20Valid
            (if (not (Health|IsDBNO InteractingPawn))
              (bind _asfort_player_state (Utilities|Casting|CastToFortPlayerState (Variables|/Script/Engine.Pawn|GetPlayerState InteractingPawn))
                (:then
                  (return (not (Math|Boolean|ORBoolean (Default|ContainsPlayerState (Variables|Default|GetSeatComponent) _asfort_player_state) (Default|IsFullandHasatLeastOneSeat (Variables|Default|GetSeatComponent)) (<= (Default|GetNumSeats (Variables|Default|GetSeatComponent)) 0) (AI|Components|NavMovement|IsCrouching (Variables|/Script/Engine.Character|GetCharacterMovement InteractingPawn))))))
                (:CastFailed
                  ;[1 0 0 0.5] ERROR CASES
                  (Fortnite|FortShippingLog self "[CanInteract] No FortPlayerState" true)))))
          (:Is%20Not%20Valid
            (Fortnite|FortShippingLog self "[CanInteract] InteractingPawn is null" true)))))
    (:then_1
      (return false))))

(fn BlueprintGetInteractionString
  (params (InteractingPawn :type Fort%20Pawn%20Object%20Reference) (InteractionBeingAttempted :type EInteractionBeingAttempted%20Enum :default "FirstInteraction"))
  (return "NSLOCTEXT(\"[21C9A896531984B20EFBD2288834CE4A]\", \"B3D104A3466B7AFD302F5A8E2D7DFE54\", \"Sit Down\")"))

(fn GetConnectedTV
  (Variables|Assets|GetConnectedTV
    (:then
      _)
    (:else
      (bind _asbp_group_hug_wide_tv_a (Utilities|Casting|CastToBP_GroupHug_WideTv_A (Utilities|ResolveSoftReference (Variables|Assets|GetTVSoftReference)))
        (:then
          (Variables|Assets|SetConnectedTV _asbp_group_hug_wide_tv_a))
        (:CastFailed
          (bind _outactors (Collision|BoxOverlapActors (Transformation|GetActorLocation) "1028.000000,1028.000000,768.000000" (Utilities|Array|MakeArray "ObjectTypeQuery1") "/WildEstate/Environment/Props/GroupHug/GroupHug_WideTv_A/Blueprints/BP_GroupHug_WideTv_A.BP_GroupHug_WideTv_A_C" :!class /Script/Engine.KismetSystemLibrary)) ; check within 2 grid spaces
          (bind _asbp_group_hug_wide_tv_a_2 (Utilities|Casting|CastToBP_GroupHug_WideTv_A (Utilities|Array|Get(acopy) _outactors))
            (:then
              (Variables|Assets|SetConnectedTV _asbp_group_hug_wide_tv_a_2))
            (:CastFailed
              (return 0)))))))
  (return (Variables|Assets|GetConnectedTV)))

(fn GetCouchMesh
  (return (Variables|Debug|GetStaticMeshComponent)))

(fn UpdateNoBuildZone)

(fn ConstructionScript)
