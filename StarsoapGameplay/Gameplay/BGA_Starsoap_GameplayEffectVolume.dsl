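; Starsoap gameplay effect volume for a hotel area, running a target hunt through warmup, neutral, active and inactive states.
; Overlaps apply or remove state- and target-based GameplayEffects, music and proximity-voice attenuation.
; An elimination in neutral makes the killer the target; the target's death rewards players, starts a cooldown and restores neutral.
; The state's OnRep swaps day-sequence lighting, alarm lights, music and occupant effects, and refreshes minimap icons and pings.
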(event EventEndPlay
  (params (EndPlayReason :type EEndPlayReason%20Enum :default "Destroyed"))
  (bind _appliedteamingeffects (Variables|Default|GetAppliedTeamingEffects))
  (Networking|SwitchHasAuthority
    (:Authority
      (bind (_keypin _valuepin) (Utilities|Map|ForEachLoop(Map) _appliedteamingeffects)
        (:then
          (Utilities|IsValid _keypin
            (:Is%20Valid
              (GameplayEffects|RemoveActiveGameplayEffect _keypin _valuepin -1))))
        (:CompletedPin
          (Utilities|Map|Clear _appliedteamingeffects))))))

(event EventBeginPlay
  ; Initial Minimap Icon
  (|CallParentFunction)
  (bind _returnvalue (Game|SpawnActorfromClass (Variables|UI|GetMinimapIcon_Default) (Math|Conversions|ToTransform(Vector) (Transformation|GetActorLocation)) "Undefined" "MultiplyWithRoot" 0 0))
  (Variables|UI|SetMinimap_Data _returnvalue))

(event Custom|AudioOn
  (bind _sc_fn_reload_hotel_music_loop_cue (Variables|Default|GetSC_FN_Reload_Hotel_Music_Loop_Cue))
  ; Audio - Turn on and off when exiting/entering volume
  (bind _asfort_player_pawn (Utilities|Casting|CastToFortPlayerPawn (Variables|Default|GetOverlapping%20Actor)))
  (if (Pawn|IsLocallyViewed _asfort_player_pawn)
    (switch Utilities|FlowControl|Switch|SwitchonEnum_Starsoap_States (Variables|Default|GetStarsoap_States)
      (:NewEnumerator0
        _)
      (:NewEnumerator1
        _)
      (:NewEnumerator2
        (if (Audio|Components|Audio|IsPlaying _sc_fn_reload_hotel_music_loop_cue) :!label _label
          (Audio|Components|Audio|FadeOut _sc_fn_reload_hotel_music_loop_cue 1.0 0.0 "Linear" :!class /Script/Engine.AudioComponent)))
      (:NewEnumerator3
        (goto _label)))
    (if (not (Audio|Components|Audio|IsPlaying _sc_fn_reload_hotel_music_loop_cue))
      (Audio|Components|Audio|FadeIn _sc_fn_reload_hotel_music_loop_cue 0.5 1.0 0.0 "Linear" :!class /Script/Engine.AudioComponent))))

(event Collision|EventActorBeginOverlap
  (params (OtherActor :type Actor%20Object%20Reference))
  (bind _returnvalue (Ability|GetAbilitySystemComponent 0))
  ; On entry apply effects if in an active state
  (Variables|Default|SetOverlapping%20Actor OtherActor)
  (if (not (Utilities|Enum|Equal(Enum) (Variables|Default|GetStarsoap_States) "NewEnumerator3"))
    (bind _asfort_player_pawn (Utilities|Casting|CastToFortPlayerPawn (Variables|Default|GetOverlapping%20Actor)))
    (Utilities|Array|AddUnique (Variables|Default|GetPlayersInArea) _asfort_player_pawn)
    (CallFunction|OnPlayerEnterArea _asfort_player_pawn)
    (Utilities|IsValid (Ability|GetFortAbilitySystemComponent _asfort_player_pawn)
      (:Is%20Valid
        (GameplayEffects|ApplyGameplayEffectToTarget :self (Ability|GetAbilitySystemComponent 0) :GameplayEffectClass (Utilities|Select :Index (Variables|Default|GetStarsoap_States) :NewEnumerator0 (Variables|GameplayEffects|GetGE_InArea_Warmup) :NewEnumerator1 (Variables|GameplayEffects|GetGE_InArea) :NewEnumerator2 (Variables|GameplayEffects|GetGE_InArea)) :Target (Ability|GetFortAbilitySystemComponent _asfort_player_pawn) :Level 0.0)
        (if (GameplayTags|HasMatchingGameplayTag _asfort_player_pawn (Variables|GameplayTags|GetT_CurrentTarget))
          (GameplayEffects|ApplyGameplayEffectToTarget :self _returnvalue :GameplayEffectClass (Variables|GameplayEffects|GetGE_TargetInArea) :Target (Ability|GetFortAbilitySystemComponent _asfort_player_pawn) :Level 0.0)
          (CallFunction|AudioOn)
          (else
            (GameplayEffects|ApplyGameplayEffectToTarget :self _returnvalue :GameplayEffectClass (Variables|GameplayEffects|GetGE_InArea_Heal) :Target (Ability|GetFortAbilitySystemComponent _asfort_player_pawn) :Level 0.0)
            (CallFunction|AudioOn)))))))

(event Collision|EventActorEndOverlap
  (params (OtherActor :type Actor%20Object%20Reference))
  (bind _sc_fn_reload_hotel_music_loop_cue (Variables|Default|GetSC_FN_Reload_Hotel_Music_Loop_Cue))
  ; On leaving the volume, cleanup prox chat and any applied GEs
  (Variables|Default|SetLeaving%20Actor OtherActor)
  (bind _asfort_player_pawn (Utilities|Casting|CastToFortPlayerPawn (Variables|Default|GetLeaving%20Actor)))
  (Utilities|Array|RemoveItem (Variables|Default|GetPlayersInArea) _asfort_player_pawn)
  (CallFunction|OnPlayerLeaveArea _asfort_player_pawn)
  (Utilities|IsValid (Ability|GetFortAbilitySystemComponent _asfort_player_pawn)
    (:Is%20Valid
      (GameplayEffects|RemoveActiveGameplayEffectbySourceEffect (Ability|GetFortAbilitySystemComponent _asfort_player_pawn) (Variables|GameplayEffects|GetGE_InArea) 0 -1)
      (GameplayEffects|RemoveActiveGameplayEffectbySourceEffect (Ability|GetFortAbilitySystemComponent _asfort_player_pawn) (Variables|GameplayEffects|GetGE_InArea_Heal) 0 -1)
      (GameplayEffects|RemoveActiveGameplayEffectbySourceEffect (Ability|GetFortAbilitySystemComponent _asfort_player_pawn) (Variables|GameplayEffects|GetGE_InArea_Warmup) 0 -1)
      (GameplayEffects|RemoveActiveGameplayEffectbySourceEffect (Ability|GetFortAbilitySystemComponent _asfort_player_pawn) (Variables|GameplayEffects|GetGE_BlockWeaponSwap) 0 -1)
      (GameplayEffects|RemoveActiveGameplayEffectbySourceEffect (Ability|GetFortAbilitySystemComponent _asfort_player_pawn) (Variables|GameplayEffects|GetGE_TargetInArea) 0 -1)
      ;; Turn Off Audio
      (if (Pawn|IsLocallyViewed _asfort_player_pawn)
        (if (Audio|Components|Audio|IsPlaying _sc_fn_reload_hotel_music_loop_cue)
          (Audio|Components|Audio|FadeOut _sc_fn_reload_hotel_music_loop_cue 1.0 0.0 "Linear" :!class /Script/Engine.AudioComponent))))))

(event Custom|PlayerEliminated
  (params (Player_Causer :type Fort%20Player%20Pawn%20Object%20Reference) (Player_Eliminated :type Fort%20Player%20Pawn%20Object%20Reference))
  (bind _returnvalue (Ability|GetFortAbilitySystemComponent Player_Causer))
  ; When a player is eliminated, check which player caused the elimination and set them as the target.
  (if (Variables|Default|GetCooldown_Active)
    _
    (elif (Utilities|Enum|Equal(Enum) (Variables|Default|GetStarsoap_States) "NewEnumerator1")
      (bind _output_get (Variables|Default|SetCurrentTarget Player_Causer))
      (GameplayEffects|ApplyGameplayEffectToTarget :self _returnvalue :GameplayEffectClass "/StarsoapGameplay/Gameplay/Abilities/GE_Starsoap_Active_Granter.GE_Starsoap_Active_Granter_C" :Target _returnvalue :Level 0.0)
      (GameplayEffects|ApplyGameplayEffectToTarget :self _returnvalue :GameplayEffectClass (Variables|GameplayEffects|GetGE_TargetInArea) :Target _returnvalue :Level 0.0)
      (EventDispatchers|BindEventtoOnDied _output_get (EventDispatchers|CreateEvent :!event Custom|TargetEliminated) :!class /Script/FortniteGame.FortPawn)
      (Utilities|FlowControl|DelayUntilNextFrame)
      (Variables|Default|SetStarsoap_States "NewEnumerator2"))))

(event Custom|TargetEliminated
  (params (DamagedActor :type Actor%20Object%20Reference) (Damage :type Float%20%28single-precision%29) (InstigatedBy :type Controller%20Object%20Reference) (DamageCauser :type Actor%20Object%20Reference) (HitLocation :type Vector) (FHitComponent :type Primitive%20Component%20Object%20Reference) (BoneName :type Name) (Momentum :type Vector))
  ; On target eliminted, clear current target and reset to visiuals.
  (GameplayEffects|ApplyGameplayEffectToTarget :self (Ability|GetAbilitySystemComponent 0) :GameplayEffectClass "/StarsoapGameplay/Gameplay/Abilities/GE_Starsoap_PlayerBonusElim.GE_Starsoap_PlayerBonusElim_C" :Target (Ability|GetAbilitySystemComponent InstigatedBy) :Level 0.0)
  (EventDispatchers|UnbindEventfromOnDied (Variables|Default|GetCurrentTarget) (delegate Custom|TargetEliminated) :!class /Script/FortniteGame.FortPawn)
  ;; Start Cooldown, used to prevent players from instantly becoming the target if they are inside hotel
  (Variables|Default|SetCooldown_Active true)
  (bind _returnvalue (Utilities|Time|SetTimerbyEvent (delegate Custom|TargetCooldown) (ScalableFloat|GetValueatLevel (Variables|Hotfix|GetActiveTarget_Cooldown))))
  (Variables|Default|SetCooldown_TimerHandle _returnvalue)
  (if (not (Utilities|Enum|Equal(Enum) (Variables|Default|GetStarsoap_States) "NewEnumerator3"))
    (for _array_element (Variables|Default|GetPlayersInArea)
      (if (not (GameplayTags|HasAnyMatchingGameplayTags _array_element (Variables|GameplayTags|GetT_BonusElim)))
        (GameplayEffects|ApplyGameplayEffectToTarget :self (Ability|GetAbilitySystemComponent 0) :GameplayEffectClass (Variables|GameplayEffects|GetGE_TargetEliminated) :Target (Ability|GetAbilitySystemComponent _array_element) :Level 0.0)))
    (Utilities|FlowControl|DelayUntilNextFrame)
    (Variables|Default|SetStarsoap_States "NewEnumerator1")))

(event Custom|TargetCooldown
  ; On target eliminted, clear current target and reset to visiuals.
  ;; Start Cooldown, used to prevent players from instantly becoming the target if they are inside hotel
  (Variables|Default|SetCooldown_Active false))

(fn OnPlayerEnterArea
  (params (PlayerPawn :type Fort%20Player%20Pawn%20Object%20Reference))
  (bind _returnvalue (Fortnite|GetFortPlayerControllerfromActor PlayerPawn))
  (bind _inareaattenuation (Variables|Default|GetInAreaAttenuation))
  (Utilities|IsValid PlayerPawn
    (:Is%20Valid
      (Networking|SwitchHasAuthority
        (:Authority
          ; Update Proximity chat to be attenuated
          (Utilities|IsValid _inareaattenuation
            (:Is%20Valid
              (Utilities|IsValid _returnvalue
                (:Is%20Valid
                  (Utilities|Map|Add (Variables|Default|GetDefaultAttenuationPerPlayer) _returnvalue (Variables|/Script/Engine.SoundAttenuation|GetAttenuation (Voice|GetProximityVoiceChatDistanceAttenuationSettings _returnvalue)))
                  (Voice|SetProximityVoiceChatSettings _returnvalue (Variables|/Script/Engine.SoundAttenuation|GetAttenuation _inareaattenuation)))))))))))

(fn OnPlayerLeaveArea
  (params (PlayerPawn :type Fort%20Player%20Pawn%20Object%20Reference))
  (bind _returnvalue (Fortnite|GetFortPlayerControllerfromActor PlayerPawn))
  (bind _defaultattenuationperplayer (Variables|Default|GetDefaultAttenuationPerPlayer))
  (bind (_value _returnvalue_2) (Utilities|Map|Find _defaultattenuationperplayer _returnvalue))
  (Networking|SwitchHasAuthority
    (:Authority
      ; Revert the Attenuation back to the playlist default
      (if _returnvalue_2
        (Voice|SetProximityVoiceChatSettings _returnvalue _value)))))

(fn OnRep_Starsoap_States
  (bind _currenttarget (Variables|Default|GetCurrentTarget))
  (bind _dsv_neutral (Variables|Lighting|GetDSV_Neutral))
  (bind _visualactors (Variables|Default|GetVisualActors))
  (bind _dsv_active (Variables|Lighting|GetDSV_Active))
  (bind _dsv_neutral_2 (Variables|Lighting|GetDSV_Neutral))
  (bind _dsv_active_2 (Variables|Lighting|GetDSV_Active))
  (bind _dsv_activefrontdoor (Variables|Lighting|GetDSV_ActiveFrontDoor))
  (bind _sc_fn_reload_hotel_music_loop_cue (Variables|Default|GetSC_FN_Reload_Hotel_Music_Loop_Cue))
  (bind _sc_fn_reload_hotel_music_loop_cue_2 (Variables|Default|GetSC_FN_Reload_Hotel_Music_Loop_Cue))
  (bind _dsv_neutralfrontdoor (Variables|Lighting|GetDSV_NeutralFrontDoor))
  (bind _dsv_activefrontdoor_2 (Variables|Lighting|GetDSV_ActiveFrontDoor))
  (bind _visualactors_2 (Variables|Default|GetVisualActors))
  (bind _playersinarea (Variables|Default|GetPlayersInArea))
  (bind _dsv_neutralfrontdoor_2 (Variables|Lighting|GetDSV_NeutralFrontDoor))
  (Default|CallStarsoapStateChange)
  (switch Utilities|FlowControl|Switch|SwitchonEnum_Starsoap_States (Variables|Default|GetStarsoap_States)
    (:NewEnumerator0
      (CallFunction|PingonStateSwap))
    (:NewEnumerator1
      ; Switch to neutral trate - remove Active state FX and turn on necessary FX
      (CallFunction|PingonStateSwap)
      (CallFunction|SetMiniMapIcon)
      (Utilities|IsValid _dsv_active
        (:Is%20Valid
          (Class|BDaySequenceVolumeParent|DisableDaysequenceVolume _dsv_active))
        (:Is%20Not%20Valid
          _))
      (Utilities|IsValid _dsv_activefrontdoor
        (:Is%20Valid
          (Class|BDaySequenceVolumeParent|DisableDaysequenceVolume _dsv_activefrontdoor))
        (:Is%20Not%20Valid
          _))
      (Utilities|IsValid _dsv_neutral
        (:Is%20Valid
          (Class|BDaySequenceVolumeParent|EnableDaysequenceVolume _dsv_neutral))
        (:Is%20Not%20Valid
          _))
      (Utilities|IsValid _dsv_neutralfrontdoor_2
        (:Is%20Valid
          (Class|BDaySequenceVolumeParent|EnableDaysequenceVolume _dsv_neutralfrontdoor_2))
        (:Is%20Not%20Valid
          _))
      (Rendering|Material|SetScalarParameterValue "/WildEstate/TechArt/MPCs/MPC_WildEstate_TA.MPC_WildEstate_TA" "Confidential" :!class /Script/Engine.KismetMaterialLibrary)
      (if (Utilities|Array|IsNotEmpty _visualactors)
        (for _array_element _visualactors
          (bind _asbp_glow_donut_cage_light_a_starsoap (Utilities|Casting|CastToBP_GlowDonut_CageLight_A_Starsoap _array_element))
          (Class|BPGlowDonutCageLightAStarsoap|BasilOkraDeactivateAlarm _asbp_glow_donut_cage_light_a_starsoap)))
      (for _array_element_2 (Variables|Default|GetPlayersInArea)
        (GameplayEffects|RemoveActiveGameplayEffectbySourceEffect (Ability|GetAbilitySystemComponent _array_element_2) (Variables|GameplayEffects|GetGE_InArea_Warmup) 0 -1)
        (GameplayEffects|RemoveActiveGameplayEffectbySourceEffect (Ability|GetAbilitySystemComponent _array_element_2) (Variables|GameplayEffects|GetGE_BlockWeaponSwap) 0 -1)
        (GameplayEffects|ApplyGameplayEffectToTarget :self (Ability|GetAbilitySystemComponent 0) :GameplayEffectClass (Variables|GameplayEffects|GetGE_InArea) :Target (Ability|GetAbilitySystemComponent _array_element_2) :Level 0.0)
        (if (Pawn|IsLocallyViewed _array_element_2)
          (if (not (Audio|Components|Audio|IsPlaying _sc_fn_reload_hotel_music_loop_cue_2))
            (Audio|Components|Audio|FadeIn _sc_fn_reload_hotel_music_loop_cue_2 0.5 1.0 0.0 "Linear" :!class /Script/Engine.AudioComponent))))
      (CallFunction|PingonStateSwap))
    (:NewEnumerator2
      ; Switch to Active state - turn on necessary FX and remove neutral state FX
      (CallFunction|PingonStateSwap)
      (CallFunction|SetMiniMapIcon)
      (Utilities|IsValid _dsv_neutral_2
        (:Is%20Valid
          (Class|BDaySequenceVolumeParent|DisableDaysequenceVolume _dsv_neutral_2))
        (:Is%20Not%20Valid
          _))
      (Utilities|IsValid _dsv_neutralfrontdoor
        (:Is%20Valid
          (Class|BDaySequenceVolumeParent|DisableDaysequenceVolume _dsv_neutralfrontdoor))
        (:Is%20Not%20Valid
          _))
      (Utilities|IsValid _dsv_active_2
        (:Is%20Valid
          (Class|BDaySequenceVolumeParent|EnableDaysequenceVolume _dsv_active_2))
        (:Is%20Not%20Valid
          _))
      (Utilities|IsValid _dsv_activefrontdoor_2
        (:Is%20Valid
          (Class|BDaySequenceVolumeParent|EnableDaysequenceVolume _dsv_activefrontdoor_2))
        (:Is%20Not%20Valid
          _))
      (Rendering|Material|SetScalarParameterValue "/WildEstate/TechArt/MPCs/MPC_WildEstate_TA.MPC_WildEstate_TA" "Confidential" 1.0 :!class /Script/Engine.KismetMaterialLibrary)
      (if (Utilities|Array|IsNotEmpty _visualactors_2)
        (for _array_element_3 _visualactors_2
          (bind _asbp_glow_donut_cage_light_a_starsoap_2 (Utilities|Casting|CastToBP_GlowDonut_CageLight_A_Starsoap _array_element_3))
          (Class|BPGlowDonutCageLightAStarsoap|BasilOkraActivateAlarm _asbp_glow_donut_cage_light_a_starsoap_2)))
      (Utilities|IsValid _currenttarget
        (:Is%20Valid
          (GameplayEffects|RemoveActiveGameplayEffectbySourceEffect (Ability|GetAbilitySystemComponent _currenttarget) (Variables|GameplayEffects|GetGE_InArea_Heal) 0 -1))
        (:Is%20Not%20Valid
          _))
      (if (Utilities|Array|IsNotEmpty _playersinarea)
        (for _array_element_4 _playersinarea
          (GameplayEffects|RemoveActiveGameplayEffectbySourceEffect (Ability|GetAbilitySystemComponent _array_element_4) (Variables|GameplayEffects|GetGE_InArea) 0 -1)
          (GameplayEffects|ApplyGameplayEffectToTarget :self (Ability|GetAbilitySystemComponent 0) :GameplayEffectClass (Variables|GameplayEffects|GetGE_InArea) :Target (Ability|GetAbilitySystemComponent _array_element_4) :Level 0.0)))
      (CallFunction|PingonStateSwap)
      (if (Audio|Components|Audio|IsPlaying _sc_fn_reload_hotel_music_loop_cue)
        (Audio|Components|Audio|FadeOut _sc_fn_reload_hotel_music_loop_cue 1.0 0.0 "Linear" :!class /Script/Engine.AudioComponent)))
    (:NewEnumerator3
      ; Turn everything off in inactive state
      (CallFunction|SetMiniMapIcon)
      (if (Utilities|Array|IsNotEmpty (Variables|Default|GetVisualActors))
        (for _array_element_5 (Variables|Default|GetVisualActors)
          (bind _asbp_glow_donut_cage_light_a_starsoap_3 (Utilities|Casting|CastToBP_GlowDonut_CageLight_A_Starsoap _array_element_5))
          (Class|BPGlowDonutCageLightAStarsoap|BasilOkraActivateAlarm _asbp_glow_donut_cage_light_a_starsoap_3)))
      (for _array_element_6 (Variables|Default|GetPlayersInArea)
        (GameplayEffects|RemoveActiveGameplayEffectbySourceEffect (Ability|GetAbilitySystemComponent _array_element_6) (Variables|GameplayEffects|GetGE_InArea_Warmup) 0 -1)
        (GameplayEffects|RemoveActiveGameplayEffectbySourceEffect (Ability|GetAbilitySystemComponent _array_element_6) (Variables|GameplayEffects|GetGE_BlockWeaponSwap) 0 -1)
        (CallFunction|OnPlayerLeaveArea _array_element_6))
      (Audio|Components|Audio|FadeOut (Variables|Default|GetSC_FN_Reload_Hotel_Music_Loop_Cue) 1.0 0.0 "Linear" :!class /Script/Engine.AudioComponent))))

(fn PingOnStateSwap
  (bind (_markerdisplay_displayname _markerdisplay_icon _markerdisplay_customindicatorclass _markerdisplay_customdetailsclass _markerdisplay_primarycolor _markerdisplay_sound _markerdisplay_battachsound _markerdisplay_screenclamping _markerdisplay_specialcull _markerdisplay_specialdotsizethreshold _markerdisplay_specialdistance _markerdisplay_customtimetolive _markerdisplay_additionaldata) (Utilities|Struct|BreakMarkedActorDisplayInfo (Variables|GetMarkerDisplay_DisplayName)))
  (Networking|SwitchHasAuthority
    (:Authority
      (for _array_element (Fortnite|GetAllFortPlayerControllers true true)
        (Utilities|IsValid _array_element
          (:Is%20Valid
            (Utilities|IsValid (Actor|GetComponentbyClass _array_element "/Script/FortniteGame.AthenaMarkerComponent")
              (:Is%20Valid
                (Variables|SetMarkerDisplay_DisplayName (Utilities|Struct|MakeMarkedActorDisplayInfo _markerdisplay_displayname (Utilities|Select :Index (Variables|Default|GetStarsoap_States) :NewEnumerator0 (Variables|UI|GetPingIcon_Warmup) :NewEnumerator1 (Variables|UI|GetPingIcon_Neutral) :NewEnumerator2 (Variables|UI|GetPingIcon_Target%20Active) :NewEnumerator3 (Variables|UI|GetPingIcon_Warmup)) _markerdisplay_customindicatorclass _markerdisplay_customdetailsclass _markerdisplay_primarycolor _markerdisplay_sound _markerdisplay_battachsound _markerdisplay_screenclamping _markerdisplay_specialcull _markerdisplay_specialdotsizethreshold _markerdisplay_specialdistance _markerdisplay_customtimetolive _markerdisplay_additionaldata))
                (MarkerComponent|MarkActoronServer (Actor|GetComponentbyClass _array_element "/Script/FortniteGame.AthenaMarkerComponent") self false true)
                (return)))))))))

(fn SetMiniMapIcon
  (Networking|SwitchHasAuthority
    (:Authority
      (Actor|DestroyActor (Variables|UI|GetMinimap_Data))
      (switch Utilities|FlowControl|Switch|SwitchonEnum_Starsoap_States (Variables|Default|GetStarsoap_States)
        (:NewEnumerator0
          (bind _returnvalue (Game|SpawnActorfromClass (Variables|UI|GetMinimapIcon_Default) (Math|Conversions|ToTransform(Vector) (Transformation|GetActorLocation)) "Undefined" "MultiplyWithRoot" 0 0))
          (Variables|UI|SetMinimap_Data _returnvalue)
          (return))
        (:NewEnumerator1
          (bind _returnvalue_2 (Game|SpawnActorfromClass (Variables|UI|GetMinimapIcon_Neutral) (Math|Conversions|ToTransform(Vector) (Transformation|GetActorLocation)) "Undefined" "MultiplyWithRoot" 0 0))
          (Variables|UI|SetMinimap_Data _returnvalue_2)
          (return))
        (:NewEnumerator2
          (bind _returnvalue_3 (Game|SpawnActorfromClass (Variables|UI|GetMinimapIcon_Active) (Math|Conversions|ToTransform(Vector) (Transformation|GetActorLocation)) "Undefined" "MultiplyWithRoot" 0 0))
          (Variables|UI|SetMinimap_Data _returnvalue_3)
          (return))))))

(fn ConstructionScript)
