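; GameState component for the Starsoap gameplay feature that advances a replicated Starsoap state as the storm progresses.
; Once the playlist is ready, it listens for the safe zone indicator event, caches the level's Starsoap effect volume, and binds to zone state changes on authority.
; Phase 1 shrinking moves the initial state forward, and phase 6 onward moves either middle state to the last.
; The state's OnRep copies it to the volume.
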(event EventBeginPlay
  (|CallParentFunction)
  (Development|PrintString "BeginPlay" true true "(R=0.000000,G=0.660000,B=1.000000,A=1.000000)" 5.0)
  (Class|FortAsyncActionCurrentPlaylistReady|PlaylistDataReadyAsync
    (:OnReady
      (Development|PrintString "Initialize" true true "(R=0.000000,G=0.660000,B=1.000000,A=1.000000)" 5.0)
      (bind _returnvalue (GameplayEventRouter|GetEventRouter "/Script/FortniteGame.FortGameState" (Components|GetOwner)))
      (bind (_asynctaskproxy _payload) (GameplayEventRouter|StartListeningToStatefulEvent _returnvalue "/Script/FortniteGame.FortSafeZoneIndicatorUpdatedEvent")
        (:then
          (Variables|Default|SetSafeZoneIndicatorUpdatedAsyncTask _asynctaskproxy)
          (bind _starsoap_volume (CallFunction|GetStarsoapVolume))
          (Variables|Default|SetStarsoap_Volume _starsoap_volume))
        (:OnEventReceived
          _)
        (:OnSavedEventStateExists
          _))
      (CallFunction|HandleSafeZoneIndicatorSpawned (Utilities|Struct|BreakFortSafeZoneIndicatorUpdatedEvent _payload)))))

(event EventTick
  (params (DeltaSeconds :type Float%20%28single-precision%29)))

(fn GetStarsoapVolume
  (Variables|Default|GetStarsoap_Volume
    (:then
      (return (Variables|Default|GetStarsoap_Volume)))
    (:else
      (bind _returnvalue (Actor|GetActorOfClass "/StarsoapGameplay/Gameplay/BGA_Starsoap_GameplayEffectVolume.BGA_Starsoap_GameplayEffectVolume_C"))
      (Utilities|IsValid _returnvalue
        (:Is%20Valid
          (return _returnvalue))))))

(fn HandleSafeZoneIndicatorSpawned
  (params (SpawnedSafeZoneIndicator :type Fort%20Safe%20Zone%20Indicator%20Object%20Reference))
  ;[0.02353 0.2078 0.1137] HandleSafeZoneIndicatorSpawned: Bind to know whenever the safe zone state changes.
  (Networking|SwitchHasAuthority
    (:Authority
      (Utilities|IsValid SpawnedSafeZoneIndicator
        (:Is%20Valid
          (SafeZone|BindEventtoSafezoneStateChangedDelegate SpawnedSafeZoneIndicator (EventDispatchers|CreateEvent :!event HandleSafeZoneStateChanged))
          ;[0.08785 0.776 0.4246] Grab this to be the context actor  so FortDataRegistry can be accessed.
          ;
          ; runs when new storm circle is spawned (after begin play)
          (Variables|Default|SetSafeZoneIndicator SpawnedSafeZoneIndicator)
          (Fortnite|FortShippingLog self "HandleSafeZoneIndicatorSpawned: Spawned.")
          (CallFunction|HandleSafeZoneUpdated (SafeZone|GetCurrentPhaseIndex SpawnedSafeZoneIndicator) (SafeZone|GetSafeZoneState SpawnedSafeZoneIndicator)))))))

(fn HandleSafeZoneStateChanged
  (params (SafeZoneIndicator :type Fort%20Safe%20Zone%20Indicator%20Object%20Reference) (SafeZoneState :type EFortSafeZoneState%20Enum))
  ;[0.02663 0.1667 0.06633] HandleSafeZoneStateChanged: Call the function that takes action based on safe zone state changes.
  (Networking|SwitchHasAuthority
    (:Authority
      (Utilities|IsValid SafeZoneIndicator
        (:Is%20Valid
          (CallFunction|HandleSafeZoneUpdated (SafeZone|GetCurrentPhaseIndex SafeZoneIndicator) SafeZoneState))))))

(fn HandleSafeZoneUpdated
  (params (PhaseIndex :type Integer) (SafeZoneState :type EFortSafeZoneState%20Enum))
  (Development|PrintString (Utilities|String|Append "Starsoap Current State" (Utilities|String|EnumtoString (Variables|Default|GetStarsoap_States))) true true "(R=0.000000,G=0.660000,B=1.000000,A=1.000000)" 1e+01)
  (switch Utilities|FlowControl|Switch|SwitchonEnum_Starsoap_States (Variables|Default|GetStarsoap_States)
    (:NewEnumerator0
      (if (and (== PhaseIndex 1) (Utilities|Enum|Equal(Enum) SafeZoneState "Shrinking"))
        (Variables|Default|SetStarsoap_States "NewEnumerator1")))
    (:NewEnumerator1
      (if (>= PhaseIndex 6) :!label _label
        (Variables|Default|SetStarsoap_States "NewEnumerator3")))
    (:NewEnumerator2
      (goto _label))
    (:NewEnumerator3
      _))
  (Fortnite|FortShippingLog self (Utilities|String|Append "HandleSafeZoneUpdated: Phase Index is " (Utilities|String|ToString(Integer) PhaseIndex) " ; SafeZoneState is " (Utilities|String|EnumtoString SafeZoneState))))

(fn OnRep_Starsoap_States
  (Variables|/StarsoapGameplay/Gameplay/BGA_Starsoap_GameplayEffectVolume.BGA_Starsoap_GameplayEffectVolume_C|SetStarsoap_States (Variables|Default|GetStarsoap_States) (Variables|Default|GetStarsoap_Volume)))

(fn OnRep_Starsoap_Volume)
