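; Cage light alarm prop that toggles meltdown warning visuals and optional audio, and destroys itself with its linked actor.
; Overridable activate/deactivate events switch a radial Niagara effect and call an emissive tint change, whose branch is fixed true, so it always sets white.
; Audio plays an alarm, selectable by index, while a looping timer spawns randomly delayed meltdown one-shots and pipe hisses.
; Deactivation fades the alarm and clears the timer.
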(event EventBeginPlay
  (|CallParentFunction))

(event Collision|EventActorBeginOverlap
  (params (OtherActor :type Actor%20Object%20Reference)))

(event EventTick
  (params (DeltaSeconds :type Float%20%28single-precision%29)))

(event Custom|BasilOkra_ActivateAlarm
  ;[1 1 1] Override in children for different FX
  (CallFunction|ToggleVisuals true)
  (if (Variables|Default|GetbShouldPlayAudio)
    (CallFunction|StartAlarmAudio)))

(event Custom|BasilOkra_DeactivateAlarm
  ;[1 1 1] Override in children for different FX
  (CallFunction|ToggleVisuals false)
  (if (Variables|Default|GetbShouldPlayAudio)
    (CallFunction|StopAlarmAudio)))

(event Custom|StartAlarmAudio
  ;[0.7866 0.008808 0.008116 0.5] Meltdown SFX Audio
  (bind _audio (Variables|Default|GetAudio)
    (:then
      (Components|Activation|Activate _audio false)
      (Audio|Components|Audio|Play _audio 0.0)
      (bind _returnvalue (Utilities|Time|SetTimerbyFunctionName 0 "PlayMeltdownOneShot" 2.0 true))
      (Variables|Default|SetAudioMeltdownHandle _returnvalue))
    (:else)))

(event Custom|StopAlarmAudio
  ;[0.7866 0.008808 0.008116 0.5] Meltdown SFX Audio
  (bind _audio (Variables|Default|GetAudio)
    (:then
      (Audio|Components|Audio|FadeOut _audio 0.1 0.0 "Linear" :!class /Script/Engine.AudioComponent)
      (Utilities|Time|ClearandInvalidateTimerbyHandle (Variables|Default|GetAudioMeltdownHandle)))
    (:else)))

(event Custom|PlayMeltdownOneShot
  (bind _audio_randomos (Variables|Default|GetAudio_RandomOS))
  ;[0.7866 0.008808 0.008116 0.5] Meltdown SFX Audio
  (Utilities|FlowControl|Sequence
    (:then_0
      (Utilities|FlowControl|Delay (Math|Random|RandomFloatinRange 1.0 8.0))
      (bind _element (ArrayMacros|GetRandomArrayElement (Variables|Default|GetMeltdown%20Random%20OS))
        (:%20%20
          (Audio|SpawnSoundAttached _element _audio_randomos))))
    (:then_1
      (Utilities|FlowControl|Delay (Math|Random|RandomFloatinRange 8.0 15.0))
      (Audio|SpawnSoundAttached (Variables|Default|GetMeltdown%20Pipe%20Hiss%20OS) _audio_randomos))))

(event Custom|SetAudioByIndex
  (params (SoundIndex :type Integer))
  (switch int SoundIndex
    (:0
      (Audio|Components|Audio|SetSound (Variables|Default|GetAudio) "/WildEstate_BasilOkra/Assets/Audio/SFX/SC_BasilOkra_Meltdown_Alarm_Generic.SC_BasilOkra_Meltdown_Alarm_Generic"))
    (:1
      (Audio|Components|Audio|SetSound (Variables|Default|GetAudio) "/WildEstate_BasilOkra/Assets/Audio/SFX/IP_MSS_BasilOkra_Meltdown_Alarm_01.IP_MSS_BasilOkra_Meltdown_Alarm_01"))
    (:2
      (Audio|Components|Audio|SetSound (Variables|Default|GetAudio) "/WildEstate_BasilOkra/Assets/Audio/SFX/IP_MSS_BasilOkra_Meltdown_Alarm_02.IP_MSS_BasilOkra_Meltdown_Alarm_02"))
    (:3
      (Audio|Components|Audio|SetSound (Variables|Default|GetAudio) "/WildEstate_BasilOkra/Assets/Audio/SFX/IP_MSS_BasilOkra_Meltdown_Alarm_03.IP_MSS_BasilOkra_Meltdown_Alarm_03"))))

(event OnLinkedActorDestroyed(FortLinkToActor)
  (params (DamageCauser :type Actor%20Object%20Reference))
  (Actor|DestroyActor self))

(fn ToggleVisuals
  (params (Enabled :type Boolean))
  (Components|Activation|SetActive (Variables|Default|GetNS_Starsoap_Alarm_Radial) Enabled Enabled)
  (if true
    (Rendering|Material|SetVectorParameterValue (Class|BTechArtGlobalLibrary|GetDynamicMaterialInstance (Variables|Debug|GetStaticMeshComponent)) "SFX_Emissive_Tint(Expand)_1" "(R=1.000000,G=1.000000,B=1.000000,A=1.000000)" :!class /Script/Engine.MaterialInstanceDynamic)
    (else
      (Rendering|Material|SetVectorParameterValue (Class|BTechArtGlobalLibrary|GetDynamicMaterialInstance (Variables|Debug|GetStaticMeshComponent)) "SFX_Emissive_Tint(Expand)_1" "(R=0.000000,G=0.000000,B=0.000000,A=1.000000)" :!class /Script/Engine.MaterialInstanceDynamic))))

(fn ConstructionScript
  (Variables|Mesh|SetOverrideStaticMeshMobility true))
