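; BP_Starsoap_NoBuildZone is a no-build zone actor whose lifetime can be tied to another actor.
; On BeginPlay, the authority checks that LinkLifetimeToActor is valid and binds to that actor's OnDestroyed event.
; When the linked actor is destroyed, the zone destroys itself.
; The overlap, tick and construction script are empty, so this graph holds no build-blocking logic.
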(event EventBeginPlay
  (bind _linklifetimetoactor (Variables|Default|GetLinkLifetimeToActor))
  (Networking|SwitchHasAuthority
    (:Authority
      (Utilities|IsValid _linklifetimetoactor
        (:Is%20Valid
          (Game|AssignOnDestroyed _linklifetimetoactor (delegate Custom|OnDestroyed_Event)))))))

(event Collision|EventActorBeginOverlap
  (params (OtherActor :type Actor%20Object%20Reference)))

(event Custom|OnDestroyed_Event
  (params (DestroyedActor :type Actor%20Object%20Reference))
  (Actor|DestroyActor))

(event EventTick
  (params (DeltaSeconds :type Float%20%28single-precision%29)))

(fn ConstructionScript)
